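; BP_Figment_AudioVolumeManager is an actor Blueprint that switches the Figment landscape's ambient audio between an initial and an event Audio Gameplay Volume.
; On BeginPlay it waits for match-ready through an async action, then checks whether the "sp" calendar event is active.
; If active, it enables the event volume and disables the initial one; otherwise the reverse, only when both references are valid.
; Tick and overlap events are empty.
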(event EventBeginPlay
  (bind _initial_ambient_volume (Variables|Default|GetInitial%20Ambient%20Volume))
  (bind _event_ambient_volume (Variables|Default|GetEvent%20Ambient%20Volume))
  (Class|FortAsyncActionAuthorityMatchReady|MatchReadyAsync)
  (bind _biseventactive (CalendarEvents|GetCalendarEventInformation "sp"))
  (if _biseventactive
    (Utilities|IsValid _event_ambient_volume
      (:Is%20Valid
        (Utilities|IsValid _initial_ambient_volume
          (:Is%20Valid
            (AudioGameplay|SetEnabled _event_ambient_volume true :!class /Script/AudioGameplayVolume.AudioGameplayVolume)
            (AudioGameplay|SetEnabled _initial_ambient_volume false :!class /Script/AudioGameplayVolume.AudioGameplayVolume)))))
    (else
      (Utilities|IsValid _initial_ambient_volume
        (:Is%20Valid
          (Utilities|IsValid _event_ambient_volume
            (:Is%20Valid
              (AudioGameplay|SetEnabled _initial_ambient_volume true :!class /Script/AudioGameplayVolume.AudioGameplayVolume)
              (AudioGameplay|SetEnabled _event_ambient_volume false :!class /Script/AudioGameplayVolume.AudioGameplayVolume))))))))

(event Collision|EventActorBeginOverlap
  (params (OtherActor :type Actor%20Object%20Reference)))

(event EventTick
  (params (DeltaSeconds :type Float%20%28single-precision%29)))

(fn ConstructionScript)
