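; Figment S07 landscape Blueprint that drives the "heat" look through the ColorTestAmount parameter of MPC_Landscape_S07.
; At match ready, the authority sets Zone from calendar events: 1 if HP1 is active, HP's timespan ratio if that is active, otherwise 0.
; Zone's OnRep makes remote clients push the curve-mapped value to the collection after a short delay.
; The construction script and an editor event apply the same value immediately.
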(event Custom|SetZoneEd
  ;[1 1 1] Editor Commands
  (Rendering|Material|SetScalarParameterValue "/Game/Athena/Environments/Landscape/MPC/MPC_Landscape_S07.MPC_Landscape_S07" "ColorTestAmount" (Math|Curves|GetFloatValue (Variables|Default|GetCurve) (Variables|Default|GetZone)) :!class /Script/Engine.KismetMaterialLibrary))

(event Custom|SetMPCValue
  (Utilities|FlowControl|Delay 1.0)
  (Rendering|Material|SetScalarParameterValue "/Game/Athena/Environments/Landscape/MPC/MPC_Landscape_S07.MPC_Landscape_S07" "ColorTestAmount" (Math|Curves|GetFloatValue (Variables|Default|GetCurve) (Variables|Default|GetZone)) :!class /Script/Engine.KismetMaterialLibrary))

(event EventBeginPlay
  (Class|FortAsyncActionAuthorityMatchReady|MatchReadyAsync
    (:OnReady
      (CallFunction|RefreshCalender(AuthorityOnly)))))

(event Custom|RefreshCalender%20%28Authority%20Only%29
  (Networking|SwitchHasAuthority
    (:Authority
      (bind _biseventactive (CalendarEvents|GetCalendarEventInformation "HP1"))
      (if _biseventactive
        (Variables|Default|SetZone 1.0)
        (else
          (bind (_biseventactive_2 _timesincebegin _timeuntilend _timespanratio) (CalendarEvents|GetCalendarEventInformation "HP"))
          (if _biseventactive_2
            (Variables|Default|SetZone _timespanratio)
            (else
              (Variables|Default|SetZone 0.0))))))))

(event Custom|SetZone
  (params (NewParam :type Float%20%28double-precision%29))
  (Variables|Default|SetZone NewParam))

(fn OnRep_Zone
  (Networking|SwitchHasAuthority
    (:Remote
      (CallFunction|SetMPCValue))))

(fn ConstructionScript
  (Rendering|Material|SetScalarParameterValue "/Game/Athena/Environments/Landscape/MPC/MPC_Landscape_S07.MPC_Landscape_S07" "ColorTestAmount" (Math|Curves|GetFloatValue (Variables|Default|GetCurve) (Variables|Default|GetZone)) :!class /Script/Engine.KismetMaterialLibrary))
