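; BP_SnowSetup is a Blueprint that sets the holiday snow level for the Athena landscape.
; Its construction script and BeginPlay both write the Snow_Amount variable into the Holiday_SnowAmount scalar of the MPC_Landscape material parameter collection, so the value applies in editor and at runtime.
; The Tick and overlap events are empty.
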(event EventBeginPlay
  (Rendering|Material|SetScalarParameterValue "/Game/Athena/Environments/Landscape/MPC/MPC_Landscape.MPC_Landscape" "Holiday_SnowAmount" (Variables|Default|GetSnow_Amount) :!class /Script/Engine.KismetMaterialLibrary))

(event Collision|EventActorBeginOverlap
  (params (OtherActor :type Actor%20Object%20Reference)))

(event EventTick
  (params (DeltaSeconds :type Float%20%28single-precision%29)))

(fn ConstructionScript
  (Rendering|Material|SetScalarParameterValue "/Game/Athena/Environments/Landscape/MPC/MPC_Landscape.MPC_Landscape" "Holiday_SnowAmount" (Variables|Default|GetSnow_Amount) :!class /Script/Engine.KismetMaterialLibrary))
