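; Season 7 (Figment) snow controller that drives landscape snow coverage and snow-styled minimap POI labels.
; On match ready, the authority derives Snow_Amount from calendar events and a stateful snow-progress router event.
; Remote OnRep maps it through a regular or regression curve (chosen by event SP1) into MPC_Landscape_S07, then restyles POI labels at rising thresholds when enabled.
; Construction script and editor events preview snow via the same MPC parameter.
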(event Custom|SetFullSnowEd
  ;[1 1 1] Editor Commands
  (Rendering|Material|SetScalarParameterValue "/Game/Athena/Environments/Landscape/MPC/MPC_Landscape_S07.MPC_Landscape_S07" "Holiday_SnowAmount" 1e+01 :!class /Script/Engine.KismetMaterialLibrary))

(event Custom|SetSnowEd
  ;[1 1 1] Editor Commands
  (Rendering|Material|SetScalarParameterValue "/Game/Athena/Environments/Landscape/MPC/MPC_Landscape_S07.MPC_Landscape_S07" "Holiday_SnowAmount" (Math|Curves|GetFloatValue (Variables|Default|GetCurve) (Variables|Default|GetSnow_Amount)) :!class /Script/Engine.KismetMaterialLibrary))

(event Custom|SetMPCValue
  (Utilities|FlowControl|Delay 1.0)
  (bind _biseventactive (CalendarEvents|GetCalendarEventInformation "SP1"))
  (if _biseventactive
    (Rendering|Material|SetScalarParameterValue "/Game/Athena/Environments/Landscape/MPC/MPC_Landscape_S07.MPC_Landscape_S07" "Holiday_SnowAmount" (Math|Curves|GetFloatValue (Variables|Default|GetCurve_Regression) (Variables|Default|GetSnow_Amount)) :!class /Script/Engine.KismetMaterialLibrary)
    (Development|PrintString "Using REGRESSION Curve" true true "(R=0.000000,G=0.660000,B=1.000000,A=1.000000)" 5.0)
    (else
      (Rendering|Material|SetScalarParameterValue "/Game/Athena/Environments/Landscape/MPC/MPC_Landscape_S07.MPC_Landscape_S07" "Holiday_SnowAmount" (Math|Curves|GetFloatValue (Variables|Default|GetCurve) (Variables|Default|GetSnow_Amount)) :!class /Script/Engine.KismetMaterialLibrary)
      (Development|PrintString "Using REGULAR Curve" true true "(R=0.000000,G=0.660000,B=1.000000,A=1.000000)" 5.0)))
  (CallFunction|RefreshMapLocations))

(event EventBeginPlay
  (Class|FortAsyncActionAuthorityMatchReady|MatchReadyAsync
    (:then
      (CallFunction|RefreshMapLocations))
    (:OnReady
      (CallFunction|RefreshCalender(AuthorityOnly))
      (bind (_asynctaskproxy _payload) (GameplayEventRouter|StartListeningToStatefulEvent (GameplayEventRouter|GetGlobalEventRouter) "/Figment_S07/Athena/Environments/Landscape/Blueprints/Figment_SnowProgressStructure.Figment_SnowProgressStructure")
        (:OnEventReceived
          _)
        (:OnSavedEventStateExists
          _))
      (Networking|SwitchHasAuthority
        (:Authority
          (bind _payload_progress_2_175a34fb44cef4b2b8eb8bb2e56a2cca (Utilities|Struct|BreakFigmentSnowProgressStructure _payload))
          (Variables|Default|SetSnow_Amount _payload_progress_2_175a34fb44cef4b2b8eb8bb2e56a2cca))))))

(event Custom|RefreshCalender%20%28Authority%20Only%29
  (Networking|SwitchHasAuthority
    (:Authority
      (bind _biseventactive (CalendarEvents|GetCalendarEventInformation "SP"))
      (if _biseventactive
        (Variables|Default|SetSnow_Amount 1.0)
        (else
          (bind (_biseventactive_2 _timesincebegin _timeuntilend _timespanratio) (CalendarEvents|GetCalendarEventInformation "SP1"))
          (if _biseventactive_2
            (Variables|Default|SetSnow_Amount (- 1.0 _timespanratio))
            (else
              (bind _biseventactive_3 (CalendarEvents|GetCalendarEventInformation "SP0"))
              (if _biseventactive_3
                (Variables|Default|SetSnow_Amount (ScalableFloat|GetValueasDouble (Variables|Default|GetSnowTinyProgressAmount)))
                (else
                  (Variables|Default|SetSnow_Amount 0.0))))))))))

(event Custom|SetSnow
  (params (NewParam :type Float%20%28double-precision%29))
  (Variables|Default|SetSnow_Amount NewParam))

(event Custom|RefreshMapLocations
  (if (ScalableFloat|GetValueasBool (Variables|Default|GetShouldOverwriteMinimapLabels))
    (if (not (Variables|Default|GetPendingMinimapRefresh))
      (Variables|Default|SetPendingMinimapRefresh true)
      (bind _gamestate (Class|FortAsyncActionGameStateReady|GameStateReadyAsync)
        (:OnReady
          (bind _asfort_game_state_athena (Utilities|Casting|CastToFortGameStateAthena _gamestate))
          (Variables|Default|SetBRGameState _asfort_game_state_athena)
          (Variables|Default|SetPendingMinimapRefresh false)
          ;[1 1 1] Update Minimap Text
          (Utilities|FlowControl|Sequence
            (:then_0
              (CallFunction|MakeLocationVisibleinSnow (GameplayTags|MakeLiteralGameplayTag "(TagName=\"Athena.Location.POI.PolarPeak\")"))
              (CallFunction|MakeLocationVisibleinSnow (GameplayTags|MakeLiteralGameplayTag "(TagName=\"Athena.Location.POI.FrostyFlights\")"))
              (CallFunction|MakeLocationVisibleinSnow (GameplayTags|MakeLiteralGameplayTag "(TagName=\"Athena.Location.POI.SnobbyShores\")"))
              (CallFunction|MakeLocationVisibleinSnow (GameplayTags|MakeLiteralGameplayTag "(TagName=\"Athena.Location.POI.HappyHamlet\")"))
              (CallFunction|MakeLocationVisibleinSnow (GameplayTags|MakeLiteralGameplayTag "(TagName=\"Athena.Location.POI.ShiftyShafts\")")))
            (:then_1
              (if (>= (Variables|Default|GetSnow_Amount) 0.1)
                (CallFunction|MakeLocationVisibleinSnow (GameplayTags|MakeLiteralGameplayTag "(TagName=\"Athena.Location.POI.TiltedTowers\")"))
                (CallFunction|MakeLocationVisibleinSnow (GameplayTags|MakeLiteralGameplayTag "(TagName=\"Athena.Location.POI.LuckyLanding\")"))))
            (:then_2
              (if (>= (Variables|Default|GetSnow_Amount) 0.15)
                (CallFunction|MakeLocationVisibleinSnow (GameplayTags|MakeLiteralGameplayTag "(TagName=\"Athena.Location.POI.FatalFields\")"))))
            (:then_3
              (if (>= (Variables|Default|GetSnow_Amount) 0.2)
                (CallFunction|MakeLocationVisibleinSnow (GameplayTags|MakeLiteralGameplayTag "(TagName=\"Athena.Location.POI.SaltySprings\")"))))
            (:then_4
              (if (>= (Variables|Default|GetSnow_Amount) 0.25)
                (CallFunction|MakeLocationVisibleinSnow (GameplayTags|MakeLiteralGameplayTag "(TagName=\"Athena.Location.POI.PleasantPark\")"))
                (CallFunction|MakeLocationVisibleinSnow (GameplayTags|MakeLiteralGameplayTag "(TagName=\"Athena.Location.POI.LootLake\")"))))
            (:then_5
              (if (>= (Variables|Default|GetSnow_Amount) 0.3)
                (CallFunction|MakeLocationVisibleinSnow (GameplayTags|MakeLiteralGameplayTag "(TagName=\"Athena.Location.POI.HauntedHills\")"))
                (CallFunction|MakeLocationVisibleinSnow (GameplayTags|MakeLiteralGameplayTag "(TagName=\"Athena.Location.POI.DustyDivot\")"))))
            (:then_6
              (if (>= (Variables|Default|GetSnow_Amount) 0.4)
                (CallFunction|MakeLocationVisibleinSnow (GameplayTags|MakeLiteralGameplayTag "(TagName=\"Athena.Location.POI.JunkJunction\")"))
                (CallFunction|MakeLocationVisibleinSnow (GameplayTags|MakeLiteralGameplayTag "(TagName=\"Athena.Location.POI.LazyLinks\")"))
                (CallFunction|MakeLocationVisibleinSnow (GameplayTags|MakeLiteralGameplayTag "(TagName=\"Athena.Location.POI.TomatoTemple\")"))
                (CallFunction|MakeLocationVisibleinSnow (GameplayTags|MakeLiteralGameplayTag "(TagName=\"Athena.Location.POI.RetailRow\")"))
                (CallFunction|MakeLocationVisibleinSnow (GameplayTags|MakeLiteralGameplayTag "(TagName=\"Athena.Location.POI.ParadisePalms\")"))))
            (:then_7
              (if (>= (Variables|Default|GetSnow_Amount) 0.75)
                (CallFunction|MakeLocationVisibleinSnow (GameplayTags|MakeLiteralGameplayTag "(TagName=\"Athena.Location.POI.RiskyReels\")"))
                (CallFunction|MakeLocationVisibleinSnow (GameplayTags|MakeLiteralGameplayTag "(TagName=\"Athena.Location.POI.WailingWoods\")"))
                (CallFunction|MakeLocationVisibleinSnow (GameplayTags|MakeLiteralGameplayTag "(TagName=\"Athena.Location.POI.LonelyLodge\")"))))))))))

(fn MakeLocationVisibleInSnow
  (params (LocationTag :type Gameplay%20Tag%20Structure%20%28by%20ref%29))
  (bind _brgamestate (Variables|Default|GetBRGameState))
  (UI|Minimap|SetLocationColorinState :self _brgamestate :LocationTag LocationTag :State "Normal")
  (UI|Minimap|SetLocationFontinState _brgamestate LocationTag (Variables|Default|GetSnowFont) "Normal"))

(fn OnRep_Snow_Amount
  (Networking|SwitchHasAuthority
    (:Remote
      (Variables|Default|SetDidEverRepNotify true)
      (CallFunction|SetMPCValue))))

(fn ConstructionScript
  (Rendering|Material|SetScalarParameterValue "/Game/Athena/Environments/Landscape/MPC/MPC_Landscape_S07.MPC_Landscape_S07" "Holiday_SnowAmount" (Math|Curves|GetFloatValue (Variables|Default|GetCurve) (Variables|Default|GetSnow_Amount)) :!class /Script/Engine.KismetMaterialLibrary))
